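# First-person sound cue for dirt sliding, driven by a SlidingIntensity parameter and routed to player foley.
# A mixer layers a random one-shot dirt "Stop" sound, with varied pitch and volume, over a looping bed.
# The bed crossfades random dirt and debris loops, each with a short volume fade-in envelope, plus a skydive wind loop, by SlidingIntensity.
# The same parameter also continuously modulates the bed's overall volume.
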SoundCue /Game/Athena/Sounds/Character/Sliding/Sliding_Intensity_Loop_1P_Cue
root=Mixer_0
nodes [
 {
  id=Enveloper_0
  children [Random_1]
  properties {pitchCurve {editorCurveData {keys [{time=0}]}}, volumeCurve {editorCurveData {keys [<at 0 0>,<at 0 0>,<at 0.3 1>]}}}
 }
 {
  id=Enveloper_1
  children [Random_4]
  properties {pitchCurve {editorCurveData {keys [{time=0}]}}, volumeCurve {editorCurveData {keys [<at 0 0>,<at 0 0>,<at 0.3 1>]}}}
 }
 {id=Mixer_0, children [Modulator_0, ModulatorContinuous_0], properties {inputVolume [1.5,1.2]}}
 {
  id=ModulatorContinuous_0
  children [ParamCrossFade_0]
  properties {volumeModulationParams {maxOutput=1.25, minOutput=0.5, parameterName=SlidingIntensity}}
 }
 {id=Modulator_0, children [Random_0], properties {pitchMax=1.1, pitchMin=1, volumeMax=1.6, volumeMin=1.5}}
 {
  id=ParamCrossFade_0
  children [Enveloper_0,Enveloper_1,WavePlayer_0]
  properties {
   paramName=SlidingIntensity
   crossFadeInput [
    {fadeInDistanceEnd=0, fadeInDistanceStart=0, fadeOutDistanceEnd=1, fadeOutDistanceStart=1, volume=1.2}
    {fadeInDistanceEnd=1, fadeInDistanceStart=0.5, fadeOutDistanceEnd=1, fadeOutDistanceStart=0, volume=1}
    {fadeInDistanceEnd=1, fadeInDistanceStart=0.7, fadeOutDistanceEnd=1, fadeOutDistanceStart=0, volume=0.3}
   ]
  }
 }
 {
  id=Random_0
  children [WavePlayer_1, WavePlayer_7, WavePlayer_4, WavePlayer_5, WavePlayer_6, WavePlayer_13]
  properties {weights [1,1,1,1,1,1]}
 }
 {id=Random_1, children [WavePlayer_10, WavePlayer_11, WavePlayer_12], properties {weights [1,1,1]}}
 {id=Random_4, children [WavePlayer_2,WavePlayer_3,WavePlayer_8,WavePlayer_9], properties {weights [1,1,1,1]}}
 {id=WavePlayer_0, properties {bLooping=true, soundWaveAssetPtr=/Game/Athena/Sounds/Wind_Falling/AMB_Wind_Skydive_Vertical_Loop_01}}
 {id=WavePlayer_1, properties {soundWaveAssetPtr=/Game/Athena/Sounds/Character/Sliding/Waves/Dirt/Sliding_Dirt_Stop_06}}
 {id=WavePlayer_10, properties {bLooping=true, soundWaveAssetPtr=/Game/Athena/Sounds/Character/Sliding/Waves/Dirt/Sliding_Dirt_Layer01_01}}
 {id=WavePlayer_11, properties {bLooping=true, soundWaveAssetPtr=/Game/Athena/Sounds/Character/Sliding/Waves/Dirt/Sliding_Dirt_Layer01_02}}
 {id=WavePlayer_12, properties {bLooping=true, soundWaveAssetPtr=/Game/Athena/Sounds/Character/Sliding/Waves/Dirt/Sliding_Dirt_Layer01_03}}
 {id=WavePlayer_13, properties {soundWaveAssetPtr=/Game/Athena/Sounds/Character/Sliding/Waves/Dirt/Sliding_Dirt_Stop_05}}
 {id=WavePlayer_2, properties {bLooping=true, soundWaveAssetPtr=/Game/Athena/Sounds/Character/Sliding/Waves/Dirt/Sliding_Dirt_Debris_01}}
 {id=WavePlayer_3, properties {bLooping=true, soundWaveAssetPtr=/Game/Athena/Sounds/Character/Sliding/Waves/Dirt/Sliding_Dirt_Debris_02}}
 {id=WavePlayer_4, properties {soundWaveAssetPtr=/Game/Athena/Sounds/Character/Sliding/Waves/Dirt/Sliding_Dirt_Stop_02}}
 {id=WavePlayer_5, properties {soundWaveAssetPtr=/Game/Athena/Sounds/Character/Sliding/Waves/Dirt/Sliding_Dirt_Stop_03}}
 {id=WavePlayer_6, properties {soundWaveAssetPtr=/Game/Athena/Sounds/Character/Sliding/Waves/Dirt/Sliding_Dirt_Stop_04}}
 {id=WavePlayer_7, properties {soundWaveAssetPtr=/Game/Athena/Sounds/Character/Sliding/Waves/Dirt/Sliding_Dirt_Stop_01}}
 {id=WavePlayer_8, properties {bLooping=true, soundWaveAssetPtr=/Game/Athena/Sounds/Character/Sliding/Waves/Dirt/Sliding_Dirt_Debris_03}}
 {id=WavePlayer_9, properties {bLooping=true, soundWaveAssetPtr=/Game/Athena/Sounds/Character/Sliding/Waves/Dirt/Sliding_Dirt_Debris_04}}
]
properties {
 attenuationSettings=/Game/Sounds/Attenuation/Fort_Player_Footsteps_1P_Sprint_Athena_Att
 maxDistance=2097152
 soundClassObject=&Fort_Player_Foley_1P_Athena
 volumeMultiplier=2
}
