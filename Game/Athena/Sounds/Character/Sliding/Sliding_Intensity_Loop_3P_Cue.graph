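# Third-person sliding sound cue for dirt, using the enemy foley sound class and sprint-footstep attenuation.
# A mixer plays a random, modulated one-shot from the dirt "Stop" waves alongside a looping bed.
# The SlidingIntensity parameter drives the bed's overall volume and crossfades its dirt, debris and skydive-wind layers, with debris and wind set to fade in at higher intensity.
# Envelopes ramp in the randomly chosen dirt and debris loops.
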SoundCue /Game/Athena/Sounds/Character/Sliding/Sliding_Intensity_Loop_3P_Cue
root=Mixer_6
nodes [
 {
  id=Enveloper_6
  children [Random_28]
  properties {pitchCurve {editorCurveData {keys [{time=0}]}}, volumeCurve {editorCurveData {keys [<at 0 0>,<at 0 0>,<at 0.3 1>]}}}
 }
 {
  id=Enveloper_7
  children [Random_29]
  properties {pitchCurve {editorCurveData {keys [{time=0}]}}, volumeCurve {editorCurveData {keys [<at 0 0>,<at 0 0>,<at 0.3 1>]}}}
 }
 {id=Mixer_6, children [Modulator_6, ModulatorContinuous_8], properties {inputVolume [1.5,1.2]}}
 {
  id=ModulatorContinuous_8
  children [ParamCrossFade_8]
  properties {volumeModulationParams {maxOutput=1.25, minOutput=0.5, parameterName=SlidingIntensity}}
 }
 {id=Modulator_6, children [Random_0], properties {pitchMax=1.1, pitchMin=1, volumeMax=1.6, volumeMin=1.5}}
 {
  id=ParamCrossFade_8
  children [Enveloper_6, Enveloper_7, WavePlayer_108]
  properties {
   paramName=SlidingIntensity
   crossFadeInput [
    {fadeInDistanceEnd=0, fadeInDistanceStart=0, fadeOutDistanceEnd=1, fadeOutDistanceStart=1, volume=1.2}
    {fadeInDistanceEnd=1, fadeInDistanceStart=0.5, fadeOutDistanceEnd=1, fadeOutDistanceStart=0, volume=1}
    {fadeInDistanceEnd=1, fadeInDistanceStart=0.7, fadeOutDistanceEnd=1, fadeOutDistanceStart=0, volume=0.3}
   ]
  }
 }
 {
  id=Random_0
  children [WavePlayer_3, WavePlayer_7, WavePlayer_8, WavePlayer_9, WavePlayer_10, WavePlayer_11]
  properties {weights [1,1,1,1,1,1]}
 }
 {id=Random_28, children [WavePlayer_105, WavePlayer_106, WavePlayer_107], properties {weights [1,1,1]}}
 {id=Random_29, children [WavePlayer_101, WavePlayer_102, WavePlayer_103, WavePlayer_104], properties {weights [1,1,1,1]}}
 {id=WavePlayer_10, properties {soundWaveAssetPtr=/Game/Athena/Sounds/Character/Sliding/Waves/Dirt/Sliding_Dirt_Stop_04}}
 {id=WavePlayer_101, properties {bLooping=true, soundWaveAssetPtr=/Game/Athena/Sounds/Character/Sliding/Waves/Dirt/Sliding_Dirt_Debris_01}}
 {id=WavePlayer_102, properties {bLooping=true, soundWaveAssetPtr=/Game/Athena/Sounds/Character/Sliding/Waves/Dirt/Sliding_Dirt_Debris_02}}
 {id=WavePlayer_103, properties {bLooping=true, soundWaveAssetPtr=/Game/Athena/Sounds/Character/Sliding/Waves/Dirt/Sliding_Dirt_Debris_03}}
 {id=WavePlayer_104, properties {bLooping=true, soundWaveAssetPtr=/Game/Athena/Sounds/Character/Sliding/Waves/Dirt/Sliding_Dirt_Debris_04}}
 {id=WavePlayer_105, properties {bLooping=true, soundWaveAssetPtr=/Game/Athena/Sounds/Character/Sliding/Waves/Dirt/Sliding_Dirt_Layer01_01}}
 {id=WavePlayer_106, properties {bLooping=true, soundWaveAssetPtr=/Game/Athena/Sounds/Character/Sliding/Waves/Dirt/Sliding_Dirt_Layer01_02}}
 {id=WavePlayer_107, properties {bLooping=true, soundWaveAssetPtr=/Game/Athena/Sounds/Character/Sliding/Waves/Dirt/Sliding_Dirt_Layer01_03}}
 {id=WavePlayer_108, properties {bLooping=true, soundWaveAssetPtr=/Game/Athena/Sounds/Wind_Falling/AMB_Wind_Skydive_Vertical_Loop_01}}
 {id=WavePlayer_11, properties {soundWaveAssetPtr=/Game/Athena/Sounds/Character/Sliding/Waves/Dirt/Sliding_Dirt_Stop_05}}
 {id=WavePlayer_3, properties {soundWaveAssetPtr=/Game/Athena/Sounds/Character/Sliding/Waves/Dirt/Sliding_Dirt_Stop_06}}
 {id=WavePlayer_7, properties {soundWaveAssetPtr=/Game/Athena/Sounds/Character/Sliding/Waves/Dirt/Sliding_Dirt_Stop_01}}
 {id=WavePlayer_8, properties {soundWaveAssetPtr=/Game/Athena/Sounds/Character/Sliding/Waves/Dirt/Sliding_Dirt_Stop_02}}
 {id=WavePlayer_9, properties {soundWaveAssetPtr=/Game/Athena/Sounds/Character/Sliding/Waves/Dirt/Sliding_Dirt_Stop_03}}
]
properties {
 attenuationSettings=/Game/Sounds/Attenuation/Fort_Player_Footsteps_3P_Sprint_Athena_Att
 maxDistance=3584
 soundClassObject=&Fort_Player_Foley_3P_Enemy_Athena
 volumeMultiplier=1
}
